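# Looping Halloween creaking ambience cue for the Creative radio gadget sound loops, built from two stereo forest tree-movement wood loops.
# Each looping wave runs through its own endlessly looping 120-second Enveloper, whose external volume curve makes the layers swell and dip independently.
# A Mixer blends both layers into a Modulator that applies slight random pitch variation.
# It uses tile-ambient attenuation and a concurrency override allowing one instance, blocking new ones.
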SoundCue /Game/Sounds/Creative/Gadgets/Radio/Sound_Loops/Halloween/Halloween_Creaking_01_Cue
root=Modulator_5
nodes [
 {
  id=Enveloper_10
  children [WavePlayer_14]
  properties {
   bLoop=true
   bLoopIndefinitely=true
   loopEnd=120
   pitchCurve {editorCurveData {keys [{time=0}]}}
   volumeCurve {externalCurve=/Game/Sounds/Trans_Enveloper_Curves/Random120s_04, editorCurveData {keys [<at 0 1>,<at 50 0.5>,<at 179.7518 1>]}}
  }
 }
 {
  id=Enveloper_11
  children [WavePlayer_16]
  properties {
   bLoop=true
   bLoopIndefinitely=true
   loopEnd=120
   pitchCurve {editorCurveData {keys [{time=0}]}}
   volumeCurve {
    externalCurve=/Game/Sounds/Trans_Enveloper_Curves/Random120s_06
    editorCurveData {keys [<at 0 1 RCIM_Cubic>,<at 4.3082 0.1>,<at 9.3585 0.1>,<at 10.9497 1>,<at 12.6101 1>,<at 16.0692 0.5>,<at 20.0126 0.5>,<at 27.5535 1>,<at 31.9811 1>,<at 34.5409 0.3>,<at 42.3762 0.2723>,<at 46.6478 1>,<at 51.0063 1>,<at 56.8868 1.7>,<at 60.5535 1.7>,<at 64.3585 1>,<at 68.4403 1>,<at 71.4843 0.3>,<at 76.6038 0.3>,<at 82.6226 1.5>,<at 87.5346 1.5>,<at 91.6855 1>,<at 96.5283 1>,<at 99.9182 0.3>,<at 103.4465 0.3>,<at 106.5598 1.4>,<at 110.2956 1.4>,<at 113.5472 1>,<at 117.8365 0.7>,<at 120 1>]}
   }
  }
 }
 {id=Mixer_5, children [Enveloper_10,Enveloper_11], properties {inputVolume [1,1]}}
 {id=Modulator_5, children [Mixer_5], properties {pitchMax=1.1, pitchMin=0.9}}
 {id=WavePlayer_14, properties {bLooping=true, soundWaveAssetPtr=/Game/Sounds/Trans_AmbientLoops/Wood/Forest_TreesMovementB12_Loop_Stereo}}
 {id=WavePlayer_16, properties {bLooping=true, soundWaveAssetPtr=/Game/Sounds/Trans_AmbientLoops/Wood/Forest_TreesMovementB08_Loop_Stereo}}
]
properties {
 attenuationSettings=/Game/Sounds/Attenuation/Fort_TileAmbient_Att
 bOverrideConcurrency=true
 maxDistance=10000
 soundClassObject=/Game/Sounds/FortSoundClassesAndModes/Fort_Creative_Ambient_TileMix
 volumeMultiplier=0.7
 concurrencyOverrides {maxCount=1, resolutionRule=PreventNew}
}
