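# Halloween creaking sound loop for the Creative radio gadget, playing a looping stereo wooden shack creak ambience.
# A looping wave player feeds a modulator that randomizes pitch slightly upward.
# The cue uses tile ambient attenuation and the Creative ambient tile mix sound class.
# Its concurrency override allows one instance at a time and blocks new ones.
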SoundCue /Game/Sounds/Creative/Gadgets/Radio/Sound_Loops/Halloween/Halloween_Creaking_03_Cue
root=Modulator_0
nodes [
 {id=Modulator_0, children [WavePlayer_34], properties {pitchMax=1.2, pitchMin=1}}
 {
  id=WavePlayer_34
  properties {bLooping=true, soundWaveAssetPtr=/Game/Sounds/Trans_AmbientLoops/StructureMovement/Wood_ShackCreaksInside006_Loop_Stereo}
 }
]
properties {
 attenuationSettings=/Game/Sounds/Attenuation/Fort_TileAmbient_Att
 bOverrideConcurrency=true
 maxDistance=10000
 soundClassObject=/Game/Sounds/FortSoundClassesAndModes/Fort_Creative_Ambient_TileMix
 volumeMultiplier=0.7
 concurrencyOverrides {maxCount=1, resolutionRule=PreventNew}
}
