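# Sound cue for a looping Halloween wood-creaking ambience in the Creative radio gadget's sound loop set.
# A looping wave player plays a stereo shack-interior creak loop through a modulator that slightly randomizes pitch.
# It uses tile ambient attenuation and the Creative ambient tile-mix sound class.
# A concurrency override allows one instance at a time and blocks new ones rather than stopping the current one.
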SoundCue /Game/Sounds/Creative/Gadgets/Radio/Sound_Loops/Halloween/Halloween_Creaking_04_Cue
root=Modulator_17
nodes [
 {id=Modulator_17, children [WavePlayer_37], properties {pitchMax=1.1, pitchMin=0.9}}
 {
  id=WavePlayer_37
  properties {bLooping=true, soundWaveAssetPtr=/Game/Sounds/Trans_AmbientLoops/StructureMovement/Wood_ShackCreaksInside005_Loop_Stereo}
 }
]
properties {
 attenuationSettings=/Game/Sounds/Attenuation/Fort_TileAmbient_Att
 bOverrideConcurrency=true
 maxDistance=10000
 soundClassObject=/Game/Sounds/FortSoundClassesAndModes/Fort_Creative_Ambient_TileMix
 volumeMultiplier=0.8
 concurrencyOverrides {maxCount=1, resolutionRule=PreventNew}
}
