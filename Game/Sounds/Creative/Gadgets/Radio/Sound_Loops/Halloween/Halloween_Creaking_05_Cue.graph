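# Sound cue for a looping Halloween creaking ambience in the Creative radio gadget's sound loop set.
# It loops a stereo wooden shack creak wave through a modulator that randomly lowers pitch for a slower, deeper sound.
# It uses the tile ambient attenuation and Creative ambient sound class, with a concurrency override allowing a single instance and blocking new ones.
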SoundCue /Game/Sounds/Creative/Gadgets/Radio/Sound_Loops/Halloween/Halloween_Creaking_05_Cue
root=Modulator_11
nodes [
 {id=Modulator_11, children [WavePlayer_36], properties {pitchMax=0.9, pitchMin=0.7}}
 {
  id=WavePlayer_36
  properties {bLooping=true, soundWaveAssetPtr=/Game/Sounds/Trans_AmbientLoops/StructureMovement/Wood_ShackCreaksInside009_Loop_Stereo}
 }
]
properties {
 attenuationSettings=/Game/Sounds/Attenuation/Fort_TileAmbient_Att
 bOverrideConcurrency=true
 maxDistance=10000
 soundClassObject=/Game/Sounds/FortSoundClassesAndModes/Fort_Creative_Ambient_TileMix
 volumeMultiplier=0.9
 concurrencyOverrides {maxCount=1, resolutionRule=PreventNew}
}
