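# Halloween creaking ambience loop for the Creative radio gadget, built from wood creaks over a cave room tone.
# Three looping layers each apply a random delay and pitch shift to one of two house or church creak waves.
# A quieter looping cave bed gets slight pitch variation and an endlessly looping Enveloper curve that slowly swells and dips its volume.
# Tile-ambient attenuation spatializes it, and a concurrency override blocks any second instance.
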SoundCue /Game/Sounds/Creative/Gadgets/Radio/Sound_Loops/Halloween/Halloween_Creaking_06_Cue
root=Mixer_4
nodes [
 {id=Delay_12, children [Modulator_17], properties {delayMax=5}}
 {id=Delay_13, children [Modulator_18], properties {delayMax=5}}
 {id=Delay_14, children [Modulator_19], properties {delayMax=5}}
 {
  id=Enveloper_4
  children [WavePlayer_56]
  properties {
   bLoop=true
   bLoopIndefinitely=true
   loopEnd=200
   pitchCurve {editorCurveData {keys [{time=0}]}}
   volumeCurve {editorCurveData {keys [
      <at 0 1 RCIM_Cubic>
      {arriveTangent=0.0166, interpMode=RCIM_Cubic, leaveTangent=0.0166, time=24.5854, value=1.5}
      {arriveTangent=-0.037, interpMode=RCIM_Cubic, leaveTangent=-0.037, time=30, value=1.4993}
      {arriveTangent=-0.019, interpMode=RCIM_Cubic, leaveTangent=-0.019, time=38.0671, value=1.0011}
      {arriveTangent=-0.0175, interpMode=RCIM_Cubic, leaveTangent=-0.0175, time=56.0762, value=1.0025}
      {arriveTangent=-0.0126, interpMode=RCIM_Cubic, leaveTangent=-0.0126, time=66.7409, value=0.5}
      {arriveTangent=-0.0088, interpMode=RCIM_Cubic, leaveTangent=-0.0088, time=80.0213, value=0.7}
      {arriveTangent=-0.0057, interpMode=RCIM_Cubic, leaveTangent=-0.0057, time=89.5793, value=0.3}
      {arriveTangent=0.023, interpMode=RCIM_Cubic, leaveTangent=0.023, time=115.1342, value=0.5}
      {arriveTangent=0.0384, interpMode=RCIM_Cubic, leaveTangent=0.0384, time=120, value=1}
      {arriveTangent=0.0337, interpMode=RCIM_Cubic, leaveTangent=0.0337, time=128.1342, value=0.9995}
      {arriveTangent=0.0235, interpMode=RCIM_Cubic, leaveTangent=0.0235, time=134.8218, value=1.5}
      {arriveTangent=-0.0248, interpMode=RCIM_Cubic, leaveTangent=-0.0248, time=149.4654, value=1.5}
      {arriveTangent=-0.0189, interpMode=RCIM_Cubic, leaveTangent=-0.0189, time=155.1153, value=0.9965}
      {arriveTangent=-0.014, interpMode=RCIM_Cubic, leaveTangent=-0.014, time=175.9853, value=1}
      {arriveTangent=0, interpMode=RCIM_Cubic, leaveTangent=0, time=190.5136, value=0.5}
      {arriveTangent=0.0527, interpMode=RCIM_Cubic, leaveTangent=0.0527, time=196.9706, value=1.001}
      <at 200 1 RCIM_Cubic>
     ]}}
  }
 }
 {id=Looping_12, children [Delay_12]}
 {id=Looping_13, children [Delay_13]}
 {id=Looping_14, children [Delay_14]}
 {id=Mixer_4, children [Looping_12,Looping_13,Looping_14,Modulator_16], properties {inputVolume [1,1,1.3,0.3]}}
 {id=Modulator_16, children [Enveloper_4], properties {pitchMax=1.1, pitchMin=0.9}}
 {id=Modulator_17, children [Random_12], properties {pitchMax=0.6, pitchMin=0.3}}
 {id=Modulator_18, children [Random_13], properties {pitchMax=1.2, pitchMin=1}}
 {id=Modulator_19, children [Random_14], properties {pitchMax=0.8, pitchMin=0.6}}
 {id=Random_12, children [WavePlayer_57, WavePlayer_60], properties {weights [1,1]}}
 {id=Random_13, children [WavePlayer_62, WavePlayer_63], properties {weights [1,1]}}
 {id=Random_14, children [WavePlayer_66, WavePlayer_68], properties {weights [1,1]}}
 {id=WavePlayer_56, properties {bLooping=true, soundWaveAssetPtr=/Game/Sounds/Trans_AmbientLoops/RoomTones/CaveA001_Loop_Stereo}}
 {id=WavePlayer_57, properties {soundWaveAssetPtr=/Game/Sounds/Creative/Ambiences/Wood/CreakHouseB001_Stereo}}
 {id=WavePlayer_60, properties {soundWaveAssetPtr=/Game/Sounds/Creative/Ambiences/Wood/CreakHouseB004_Stereo}}
 {id=WavePlayer_62, properties {soundWaveAssetPtr=/Game/Sounds/Creative/Ambiences/Wood/Movement_ChurchA006_Stereo}}
 {id=WavePlayer_63, properties {soundWaveAssetPtr=/Game/Sounds/Creative/Ambiences/Wood/Movement_ChurchA007_Stereo}}
 {id=WavePlayer_66, properties {soundWaveAssetPtr=/Game/Sounds/Creative/Ambiences/Wood/Movement_ChurchB008_Stereo}}
 {id=WavePlayer_68, properties {soundWaveAssetPtr=/Game/Sounds/Creative/Ambiences/Wood/Movement_ChurchB006_Stereo}}
]
properties {
 attenuationSettings=/Game/Sounds/Attenuation/Fort_TileAmbient_Att
 bOverrideConcurrency=true
 maxDistance=10000
 soundClassObject=/Game/Sounds/FortSoundClassesAndModes/Fort_Creative_Ambient_TileMix
 volumeMultiplier=0.9
 concurrencyOverrides {maxCount=1, resolutionRule=PreventNew}
}
