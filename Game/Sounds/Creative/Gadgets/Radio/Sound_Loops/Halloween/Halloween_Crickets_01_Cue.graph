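# Halloween cricket ambience loop for the Creative radio gadget, built as a Sound Cue on the ambient tile mix class.
# It mixes a looping night-animal composite, pitched slightly up, with a quieter water-sloshing loop pitched well down.
# Endlessly looping envelopers give the layers opposite volume swells, so one rises as the other dips.
# It uses tile-ambient attenuation and allows only one instance, preventing new ones.
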SoundCue /Game/Sounds/Creative/Gadgets/Radio/Sound_Loops/Halloween/Halloween_Crickets_01_Cue
root=Mixer_36
nodes [
 {
  id=Enveloper_72
  children [Modulator_24]
  properties {
   bLoop=true
   bLoopIndefinitely=true
   loopEnd=200
   pitchCurve {editorCurveData {keys [{time=0}]}}
   volumeCurve {editorCurveData {keys [<at 0 1 RCIM_Cubic>,<at 30 0.5 RCIM_Cubic>,<at 200 1 RCIM_Cubic>]}}
  }
 }
 {
  id=Enveloper_75
  children [Modulator_25]
  properties {
   bLoop=true
   bLoopIndefinitely=true
   loopEnd=200
   pitchCurve {editorCurveData {keys [{time=0}]}}
   volumeCurve {editorCurveData {keys [<at 0 0.5 RCIM_Cubic>,<at 30 1 RCIM_Cubic>,<at 200 0.5 RCIM_Cubic>]}}
  }
 }
 {id=Mixer_36, children [Enveloper_75,Enveloper_72], properties {inputVolume [0.3,1]}}
 {id=Modulator_24, children [WavePlayer_130], properties {pitchMax=1.2, pitchMin=1}}
 {id=Modulator_25, children [WavePlayer_129], properties {pitchMax=0.6, pitchMin=0.4}}
 {id=WavePlayer_129, properties {bLooping=true, soundWaveAssetPtr=/Game/Sounds/Trans_AmbientLoops/Water/WaterSloshingB004_Loop_Stereo}}
 {id=WavePlayer_130, properties {bLooping=true, soundWaveAssetPtr=/Game/Sounds/Trans_AmbientLoops/Animals/NightCompositeB007_Stereo}}
]
properties {
 attenuationSettings=/Game/Sounds/Attenuation/Fort_TileAmbient_Att
 bOverrideConcurrency=true
 maxDistance=10000
 soundClassObject=/Game/Sounds/FortSoundClassesAndModes/Fort_Creative_Ambient_TileMix
 volumeMultiplier=0.8
 concurrencyOverrides {maxCount=1, resolutionRule=PreventNew}
}
