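# Halloween drips ambient loop SoundCue for the Creative radio gadget, layering a cave room tone with a pitched-down castle room tone.
# Each looping wave runs through its own endlessly looping Enveloper, whose external reversed-bell shelf curve slowly varies volume before a Mixer blends the two.
# It uses tile-ambient attenuation and the Creative ambient sound class.
# A concurrency override allows one instance and blocks new ones.
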SoundCue /Game/Sounds/Creative/Gadgets/Radio/Sound_Loops/Halloween/Halloween_Drips_01_Cue
root=Mixer_4
nodes [
 {
  id=Enveloper_16
  children [WavePlayer_8]
  properties {
   bLoop=true
   bLoopIndefinitely=true
   loopEnd=100
   pitchCurve {editorCurveData {keys [{time=0}]}}
   volumeCurve {externalCurve=/Game/Sounds/Trans_Enveloper_Curves/BellReversed_HighShelf_100s, editorCurveData {keys [<at 0 1>,<at 50 0.5>,<at 179.7518 1>]}}
  }
 }
 {
  id=Enveloper_17
  children [Modulator_8]
  properties {
   bLoop=true
   bLoopIndefinitely=true
   loopEnd=100
   pitchCurve {editorCurveData {keys [{time=0}]}}
   volumeCurve {
    externalCurve=/Game/Sounds/Trans_Enveloper_Curves/BellReversed_LowShelf_100s
    editorCurveData {keys [<at 0 1>,<at 6.705 1.0001>,<at 25 0>,<at 50 0>,<at 80 1>,<at 88.38 1>,<at 93.33 0>,<at 112.635 0>,<at 116.43 1>,<at 120 1>]}
   }
  }
 }
 {id=Mixer_4, children [Enveloper_16,Enveloper_17], properties {inputVolume [0.7,1]}}
 {id=Modulator_8, children [WavePlayer_9], properties {pitchMax=0.7, pitchMin=0.7}}
 {id=WavePlayer_8, properties {bLooping=true, soundWaveAssetPtr=/Game/Sounds/Trans_AmbientLoops/RoomTones/CaveA003_Loop_Stereo}}
 {id=WavePlayer_9, properties {bLooping=true, soundWaveAssetPtr=/Game/Sounds/Trans_AmbientLoops/RoomTones/Castle_Generic008_Stereo}}
]
properties {
 attenuationSettings=/Game/Sounds/Attenuation/Fort_TileAmbient_Att
 bOverrideConcurrency=true
 maxDistance=10000
 soundClassObject=/Game/Sounds/FortSoundClassesAndModes/Fort_Creative_Ambient_TileMix
 volumeMultiplier=0.7
 concurrencyOverrides {maxCount=1, resolutionRule=PreventNew}
}
