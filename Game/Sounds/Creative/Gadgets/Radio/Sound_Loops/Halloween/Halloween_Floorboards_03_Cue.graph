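# Sound cue for a Halloween floorboards ambient loop, filed under the Creative radio gadget's sound loops.
# A single looping wave player plays Halloween_Floorboards_03 at reduced volume, using tile-ambient attenuation and the Creative ambient tile-mix sound class.
# A concurrency override allows one instance at a time and blocks new ones rather than stopping the current one.
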SoundCue /Game/Sounds/Creative/Gadgets/Radio/Sound_Loops/Halloween/Halloween_Floorboards_03_Cue
root=WavePlayer_0
nodes [{
  id=WavePlayer_0
  properties {bLooping=true, soundWaveAssetPtr=&/Halloween_Floorboards_03}
 }]
properties {
 attenuationSettings=/Game/Sounds/Attenuation/Fort_TileAmbient_Att
 bOverrideConcurrency=true
 maxDistance=10000
 soundClassObject=/Game/Sounds/FortSoundClassesAndModes/Fort_Creative_Ambient_TileMix
 volumeMultiplier=0.7
 concurrencyOverrides {maxCount=1, resolutionRule=PreventNew}
}
