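# Looping Halloween soundscape cue for the Creative radio gadget, built on a castle room-tone stereo loop.
# A looping wave player feeds a modulator that applies slight random pitch variation.
# It uses tile ambient attenuation and the Creative ambient tile-mix sound class.
# A concurrency override allows only one instance at a time and blocks new ones.
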SoundCue /Game/Sounds/Creative/Gadgets/Radio/Sound_Loops/Halloween/Halloween_Soundscape_01_Cue
root=Modulator_22
nodes [
 {id=Modulator_22, children [WavePlayer_0], properties {pitchMax=1.1, pitchMin=0.9}}
 {id=WavePlayer_0, properties {bLooping=true, soundWaveAssetPtr=/Game/Sounds/Trans_AmbientLoops/RoomTones/Castle_Generic008_Stereo}}
]
properties {
 attenuationSettings=/Game/Sounds/Attenuation/Fort_TileAmbient_Att
 bOverrideConcurrency=true
 maxDistance=10000
 soundClassObject=/Game/Sounds/FortSoundClassesAndModes/Fort_Creative_Ambient_TileMix
 volumeMultiplier=0.8
 concurrencyOverrides {maxCount=1, resolutionRule=PreventNew}
}
