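# Halloween voices ambience Sound Cue for the Creative radio gadget's sound loops, layering eerie one-shots over a whispering room-tone bed.
# A mixer blends one randomly picked looping castle whisper with four looping branches of girl laughs, girl singing, distant howls and chains.
# Each branch uses random delay, volume and pitch variation, with chains pitched far down.
# It uses ambient attenuation and sound class, and a concurrency override allows one instance, rejecting new ones.
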SoundCue /Game/Sounds/Creative/Gadgets/Radio/Sound_Loops/Halloween/Halloween_Voices_01_Cue
root=Mixer_2
nodes [
 {id=Delay_19, children [Modulator_21], properties {delayMax=7}}
 {id=Delay_22, children [Modulator_24], properties {delayMax=10, delayMin=3}}
 {id=Delay_25, children [Modulator_27], properties {delayMax=10}}
 {id=Delay_28, children [Modulator_31], properties {delayMax=15}}
 {id=Looping_19, children [Delay_19]}
 {id=Looping_22, children [Delay_22]}
 {id=Looping_25, children [Delay_25]}
 {id=Looping_28, children [Delay_28]}
 {id=Mixer_2, children [Looping_19,Looping_22,Looping_25,Looping_28,Random_14], properties {inputVolume [0.35,0.4,0.4,0.35,0.75]}}
 {id=Modulator_21, children [Random_9], properties {pitchMax=1, pitchMin=0.5, volumeMax=0.65, volumeMin=0.35}}
 {id=Modulator_24, children [Random_10], properties {pitchMax=1, pitchMin=0.7, volumeMax=0.5, volumeMin=0.25}}
 {id=Modulator_27, children [Random_11], properties {pitchMax=1.1, pitchMin=0.9, volumeMax=0.65, volumeMin=0.35}}
 {id=Modulator_31, children [Random_12], properties {pitchMax=0.5, pitchMin=0.2, volumeMax=0.65, volumeMin=0.35}}
 {id=Random_10, children [WavePlayer_43, WavePlayer_46], properties {weights [1,1]}}
 {id=Random_11, children [WavePlayer_48, WavePlayer_50], properties {weights [1,1]}}
 {id=Random_12, children [WavePlayer_51, WavePlayer_53], properties {weights [1,1]}}
 {id=Random_14, children [WavePlayer_61, WavePlayer_64], properties {weights [1,1]}}
 {id=Random_9, children [WavePlayer_38, WavePlayer_39, WavePlayer_40, WavePlayer_41], properties {weights [1,1,1,1]}}
 {id=WavePlayer_38, properties {soundWaveAssetPtr=/Game/Sounds/Creative/Ambiences/AbstractVoices/Voice_GirlLaughA008}}
 {id=WavePlayer_39, properties {soundWaveAssetPtr=/Game/Sounds/Creative/Ambiences/AbstractVoices/Voice_GirlLaughA005}}
 {id=WavePlayer_40, properties {soundWaveAssetPtr=/Game/Sounds/Creative/Ambiences/AbstractVoices/Voice_GirlLaughA006}}
 {id=WavePlayer_41, properties {soundWaveAssetPtr=/Game/Sounds/Creative/Ambiences/AbstractVoices/Voice_GirlLaughA007}}
 {id=WavePlayer_43, properties {soundWaveAssetPtr=/Game/Sounds/Creative/Ambiences/AbstractVoices/Voice_GirlSingsA004_Stereo}}
 {id=WavePlayer_46, properties {soundWaveAssetPtr=/Game/Sounds/Creative/Ambiences/AbstractVoices/Voice_GirlSingsA001_Stereo}}
 {id=WavePlayer_48, properties {soundWaveAssetPtr=/Game/Sounds/Creative/Ambiences/AbstractVoices/Voice_HowlDistantB006_Stereo}}
 {id=WavePlayer_50, properties {soundWaveAssetPtr=/Game/Sounds/Creative/Ambiences/AbstractVoices/Voice_HowlDistantB004_Stereo}}
 {id=WavePlayer_51, properties {soundWaveAssetPtr=/Game/Sounds/Creative/Ambiences/Metal/ChainCloseA001_Stereo}}
 {id=WavePlayer_53, properties {soundWaveAssetPtr=/Game/Sounds/Creative/Ambiences/AbstractVoices/ChainCloseA003_Stereo}}
 {id=WavePlayer_61, properties {bLooping=true, soundWaveAssetPtr=/Game/Sounds/Trans_AmbientLoops/RoomTones/Castle_WhispersB020_Stereo}}
 {id=WavePlayer_64, properties {bLooping=true, soundWaveAssetPtr=/Game/Sounds/Trans_AmbientLoops/RoomTones/Castle_WhispersB017_Stereo}}
]
properties {
 attenuationSettings=/Game/Sounds/Attenuation/Fort_TileAmbient_Att
 bOverrideConcurrency=true
 maxDistance=10000
 soundClassObject=/Game/Sounds/FortSoundClassesAndModes/Fort_Creative_Ambient_TileMix
 volumeMultiplier=0.7
 concurrencyOverrides {maxCount=1, resolutionRule=PreventNew}
}
