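# Halloween wind ambience sound cue, filed under the Creative radio gadget's sound loops.
# It mixes a looping forest atmosphere at half volume with a looping spooky wind that a Modulator randomly pitches down.
# Each layer runs through an indefinitely looping 100-second Enveloper that slowly varies volume, the forest via a shared external curve.
# The cue uses tile-ambient attenuation and overrides concurrency to one instance, preventing new ones.
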SoundCue /Game/Sounds/Creative/Gadgets/Radio/Sound_Loops/Halloween/Halloween_Wind_01_Cue
root=Mixer_0
nodes [
 {
  id=Enveloper_0
  children [WavePlayer_2]
  properties {
   bLoop=true
   bLoopIndefinitely=true
   loopEnd=100
   pitchCurve {editorCurveData {keys [{time=0}]}}
   volumeCurve {externalCurve=/Game/Sounds/Trans_Enveloper_Curves/BellReversed_HighShelf_100s, editorCurveData {keys [<at 0 1>,<at 50 0.5>,<at 179.7518 1>]}}
  }
 }
 {
  id=Enveloper_1
  children [WavePlayer_5]
  properties {
   bLoop=true
   bLoopIndefinitely=true
   loopEnd=100
   pitchCurve {editorCurveData {keys [{time=0}]}}
   volumeCurve {editorCurveData {keys [
      {arriveTangent=-0.0301, interpMode=RCIM_Cubic, leaveTangent=-0.0301, tangentMode=RCTM_User, time=-0.1375, value=1}
      <at 85 0.05 RCIM_Cubic>
      {arriveTangent=0.1575, interpMode=RCIM_Cubic, leaveTangent=0.1575, tangentMode=RCTM_User, time=100, value=1}
     ]}}
  }
 }
 {id=Mixer_0, children [Enveloper_0,Modulator_0], properties {inputVolume [0.5,1]}}
 {id=Modulator_0, children [Enveloper_1], properties {pitchMax=0.9, pitchMin=0.7}}
 {id=WavePlayer_2, properties {bLooping=true, soundWaveAssetPtr=/Game/Sounds/Trans_AmbientLoops/RoomTones/ForestAtmosphere03_Loop_Stereo}}
 {id=WavePlayer_5, properties {bLooping=true, soundWaveAssetPtr=/Game/Sounds/Trans_AmbientLoops/Wind/WindSpooky06_Loop_Stereo}}
]
properties {
 attenuationSettings=/Game/Sounds/Attenuation/Fort_TileAmbient_Att
 bOverrideConcurrency=true
 maxDistance=10000
 soundClassObject=/Game/Sounds/FortSoundClassesAndModes/Fort_Creative_Ambient_TileMix
 volumeMultiplier=0.4
 concurrencyOverrides {maxCount=1, resolutionRule=PreventNew}
}
