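# Halloween wind ambience Sound Cue, filed under Creative radio gadget sound loops.
# A mixer layers looping forest room tone with a spooky wind loop at lower input volume.
# Each layer has an indefinitely looping Enveloper using an external reversed-bell shelf curve to swell and dip its volume over long cycles.
# It uses tile-ambient attenuation and the Creative ambient sound class, with concurrency allowing one instance and blocking new ones.
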SoundCue /Game/Sounds/Creative/Gadgets/Radio/Sound_Loops/Halloween/Halloween_Wind_02_Cue
root=Mixer_14
nodes [
 {
  id=Enveloper_36
  children [WavePlayer_50]
  properties {
   bLoop=true
   bLoopIndefinitely=true
   loopEnd=100
   pitchCurve {editorCurveData {keys [{time=0}]}}
   volumeCurve {externalCurve=/Game/Sounds/Trans_Enveloper_Curves/BellReversed_HighShelf_100s, editorCurveData {keys [<at 0 1>,<at 50 0.5>,<at 179.7518 1>]}}
  }
 }
 {
  id=Enveloper_37
  children [WavePlayer_51]
  properties {
   bLoop=true
   bLoopIndefinitely=true
   loopEnd=100
   pitchCurve {editorCurveData {keys [{time=0}]}}
   volumeCurve {
    externalCurve=/Game/Sounds/Trans_Enveloper_Curves/BellReversed_LowShelf_100s
    editorCurveData {keys [<at 0 1>,<at 6.705 1.0001>,<at 25 0>,<at 50 0>,<at 80 1>,<at 88.38 1>,<at 93.33 0>,<at 112.635 0>,<at 116.43 1>,<at 120 1>]}
   }
  }
 }
 {id=Mixer_14, children [Enveloper_36,Enveloper_37], properties {inputVolume [1,0.5]}}
 {id=WavePlayer_50, properties {bLooping=true, soundWaveAssetPtr=/Game/Sounds/Trans_AmbientLoops/RoomTones/ForestAtmosphere05_Loop_Stereo}}
 {id=WavePlayer_51, properties {bLooping=true, soundWaveAssetPtr=/Game/Sounds/Trans_AmbientLoops/Wind/WindSpooky01_Loop_Stereo}}
]
properties {
 attenuationSettings=/Game/Sounds/Attenuation/Fort_TileAmbient_Att
 bOverrideConcurrency=true
 maxDistance=10000
 soundClassObject=/Game/Sounds/FortSoundClassesAndModes/Fort_Creative_Ambient_TileMix
 volumeMultiplier=0.5
 concurrencyOverrides {maxCount=1, resolutionRule=PreventNew}
}
