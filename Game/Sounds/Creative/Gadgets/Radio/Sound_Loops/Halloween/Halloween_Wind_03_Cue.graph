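# Halloween_Wind_03 ambient Sound Cue, filed under the Creative radio gadget's Halloween sound loops.
# Mixes two looping stereo castle-whisper room tones, each inside an indefinitely looping Enveloper.
# External reversed-bell curves, one high-shelf and one low-shelf, swell each layer's volume over 100-second cycles so the bed keeps shifting.
# Plays on the Creative ambient sound class with tile-ambient attenuation, capped at one instance that blocks new plays.
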SoundCue /Game/Sounds/Creative/Gadgets/Radio/Sound_Loops/Halloween/Halloween_Wind_03_Cue
root=Mixer_13
nodes [
 {
  id=Enveloper_34
  children [WavePlayer_0]
  properties {
   bLoop=true
   bLoopIndefinitely=true
   loopEnd=100
   pitchCurve {editorCurveData {keys [{time=0}]}}
   volumeCurve {externalCurve=/Game/Sounds/Trans_Enveloper_Curves/BellReversed_HighShelf_100s, editorCurveData {keys [<at 0 1>,<at 50 0.5>,<at 179.7518 1>]}}
  }
 }
 {
  id=Enveloper_35
  children [WavePlayer_1]
  properties {
   bLoop=true
   bLoopIndefinitely=true
   loopEnd=100
   pitchCurve {editorCurveData {keys [{time=0}]}}
   volumeCurve {
    externalCurve=/Game/Sounds/Trans_Enveloper_Curves/BellReversed_LowShelf_100s
    editorCurveData {keys [<at 0 1>,<at 6.705 1.0001>,<at 25 0>,<at 50 0>,<at 80 1>,<at 88.38 1>,<at 93.33 0>,<at 112.635 0>,<at 116.43 1>,<at 120 1>]}
   }
  }
 }
 {id=Mixer_13, children [Enveloper_34,Enveloper_35], properties {inputVolume [0.7,1]}}
 {id=WavePlayer_0, properties {bLooping=true, soundWaveAssetPtr=/Game/Sounds/Trans_AmbientLoops/RoomTones/Castle_WhispersD001_Stereo}}
 {id=WavePlayer_1, properties {bLooping=true, soundWaveAssetPtr=/Game/Sounds/Trans_AmbientLoops/RoomTones/Castle_WhispersB016_Stereo}}
]
properties {
 attenuationSettings=/Game/Sounds/Attenuation/Fort_TileAmbient_Att
 bOverrideConcurrency=true
 maxDistance=10000
 soundClassObject=/Game/Sounds/FortSoundClassesAndModes/Fort_Creative_Ambient_TileMix
 volumeMultiplier=0.6
 concurrencyOverrides {maxCount=1, resolutionRule=PreventNew}
}
