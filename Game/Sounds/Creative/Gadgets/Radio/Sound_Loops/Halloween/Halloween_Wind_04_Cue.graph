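# Halloween Wind 04 ambient loop cue for the Creative radio gadget, on the Creative ambient tile sound class with tile ambient attenuation.
# A Mixer blends a quiet wooden shack creak loop with a full-level dread interior ambience loop.
# Each looping wave runs through an indefinitely looping Enveloper whose external volume curve varies its level over time.
# A concurrency override allows one instance and blocks new ones.
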SoundCue /Game/Sounds/Creative/Gadgets/Radio/Sound_Loops/Halloween/Halloween_Wind_04_Cue
root=Mixer_9
nodes [
 {
  id=Enveloper_26
  children [WavePlayer_19]
  properties {
   bLoop=true
   bLoopIndefinitely=true
   loopEnd=100
   pitchCurve {editorCurveData {keys [{time=0}]}}
   volumeCurve {externalCurve=/Game/Sounds/Trans_Enveloper_Curves/Random10, editorCurveData {keys [<at 0 1>,<at 50 0.5>,<at 179.7518 1>]}}
  }
 }
 {
  id=Enveloper_27
  children [WavePlayer_30]
  properties {
   bLoop=true
   bLoopIndefinitely=true
   loopEnd=100
   pitchCurve {editorCurveData {keys [{time=0}]}}
   volumeCurve {
    externalCurve=/Game/Sounds/Trans_Enveloper_Curves/BellReversed_LowShelf_100s
    editorCurveData {keys [<at 0 1>,<at 6.705 1.0001>,<at 25 0>,<at 50 0>,<at 80 1>,<at 88.38 1>,<at 93.33 0>,<at 112.635 0>,<at 116.43 1>,<at 120 1>]}
   }
  }
 }
 {id=Mixer_9, children [Enveloper_26,Enveloper_27], properties {inputVolume [0.2,1]}}
 {
  id=WavePlayer_19
  properties {bLooping=true, soundWaveAssetPtr=/Game/Sounds/Trans_AmbientLoops/StructureMovement/Wood_ShackCreaksInside004_Loop_Stereo}
 }
 {id=WavePlayer_30, properties {bLooping=true, soundWaveAssetPtr=/Game/Sounds/Fort_Ambient_Looping/Dread_Interior/dread_amb_06Loop}}
]
properties {
 attenuationSettings=/Game/Sounds/Attenuation/Fort_TileAmbient_Att
 bOverrideConcurrency=true
 maxDistance=10000
 soundClassObject=/Game/Sounds/FortSoundClassesAndModes/Fort_Creative_Ambient_TileMix
 volumeMultiplier=0.7
 concurrencyOverrides {maxCount=1, resolutionRule=PreventNew}
}
