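# Halloween wind-whispers ambient loop cue, filed among the Creative radio gadget sound loops.
# Two looping castle whisper room tones each pass through an indefinitely looping Enveloper into a Mixer.
# The volume envelopes dip to silence at opposite ends of the same cycle, so the layers swell and fade against each other.
# It uses tile-ambient attenuation and sound class, with concurrency allowing one instance and blocking new ones.
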SoundCue /Game/Sounds/Creative/Gadgets/Radio/Sound_Loops/Halloween/Halloween_Wind_Whispers_01_Cue
root=Mixer_13
nodes [
 {
  id=Enveloper_34
  children [WavePlayer_46]
  properties {
   bLoop=true
   bLoopIndefinitely=true
   loopEnd=100
   pitchCurve {editorCurveData {keys [{time=0}]}}
   volumeCurve {editorCurveData {keys [
      <at 0 1 RCIM_Cubic>
      <at 10 0 RCIM_Cubic>
      {arriveTangent=0.0154, interpMode=RCIM_Cubic, leaveTangent=0.0154, tangentMode=RCTM_User, time=100, value=1}
     ]}}
  }
 }
 {
  id=Enveloper_35
  children [WavePlayer_47]
  properties {
   bLoop=true
   bLoopIndefinitely=true
   loopEnd=100
   pitchCurve {editorCurveData {keys [{time=0}]}}
   volumeCurve {editorCurveData {keys [
      {arriveTangent=-0.0178, interpMode=RCIM_Cubic, leaveTangent=-0.0178, tangentMode=RCTM_User, time=0, value=1}
      <at 85 0 RCIM_Cubic>
      <at 100 1 RCIM_Cubic>
     ]}}
  }
 }
 {id=Mixer_13, children [Enveloper_34,Enveloper_35], properties {inputVolume [1,1]}}
 {id=WavePlayer_46, properties {bLooping=true, soundWaveAssetPtr=/Game/Sounds/Trans_AmbientLoops/RoomTones/Castle_WhispersD003_Stereo}}
 {id=WavePlayer_47, properties {bLooping=true, soundWaveAssetPtr=/Game/Sounds/Trans_AmbientLoops/RoomTones/Castle_WhispersB003_Stereo}}
]
properties {
 attenuationSettings=/Game/Sounds/Attenuation/Fort_TileAmbient_Att
 bOverrideConcurrency=true
 maxDistance=10000
 soundClassObject=/Game/Sounds/FortSoundClassesAndModes/Fort_Creative_Ambient_TileMix
 volumeMultiplier=0.6
 concurrencyOverrides {maxCount=1, resolutionRule=PreventNew}
}
